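# Unlit translucent Niagara sprite material for the cloud layer of the LiftingAura contrail effect.
# A panning swirl-noise texture distorts the cloud texture UVs, scaled by a dynamic parameter.
# The cloud red channel blends two colors, tinted by particle color and passed through emissive adjustments.
# An alpha mask combines particle-alpha-driven power erosion, a smoothstep dissolve broken up by paint-cloud noise, and a hard Ceil cutoff on red.
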Material /Game/NeverCook/FX/FrontEndFX/S3940/M_Contrail_LiftingAura_Clouds
graph {
 properties {bUsedWithNiagaraSprites=true, blendMode=BLEND_Translucent, pixelDepthOffsetMode=PDOM_Legacy, shadingModel=MSM_Unlit, usedShadingModels=Unlit}
 nodes [
  n1 = ParticleColor()
  n2 = VectorParameter() {parameterName=Color A, defaultValue=<rgba 0.8 0.16 0.256 0>}
  n3 = VectorParameter() {parameterName=Color B, defaultValue=<rgba 1 0.334 0.445 3>}
  n4 = TextureCoordinate()
  n5 = VectorParameter() {parameterName=Distortion UV, channelNames {a=Speed V, b=Speed U, g=Tile V, r=Tile U}, defaultValue=<rgba 1 1 0 0>}
  n6 = AppendVector(A=@n5.Tile U, B=@n5.Tile U)
  n7 = TextureSampleParameter2D(UVs=Panner(Coordinate=Multiply(A=@n4, B=@n6), Speed=AppendVector(A=@n5.Speed U, B=@n5.Speed U))) {parameterName=Distortion Tex, samplerType=SAMPLERTYPE_Normal, texture=/Game/Effects/Fort_Effects/Emotes/BirdNestBlue/T_SwirlNoise}
  n8 = AppendVector(A=@n7.R, B=@n7.R)
  n9 = DynamicParameter() {defaultValue=<rgba 0 0 0.5 1>, paramNames [Dissolve, Distortion, Dissolve 02, Param4]}
  n10 = TextureSampleParameter2D(UVs=Add(A=TextureCoordinate(), B=Multiply(A=@n8, B=@n9.Distortion) {constB=0})) {parameterName=Base Tex, samplerType=SAMPLERTYPE_LinearColor, texture=/Game/Effects/Fort_Effects/Textures/T_Contrail_LiftingAura_Clouds_2x2}
  MaterialFunctionCall(Emissive=Multiply(A=@n1.RGB, B=LinearInterpolate(A=@n2.RGB, B=@n3.RGB, Alpha=@n10.R))) {materialFunction=&MF_FX_EmissiveAdjustments}
  n11 = DynamicParameter() {defaultValue=<rgba 0 0 0.5 1>, paramNames [Dissolve, Distortion, Dissolve 02, Param4]}
  n12 = Multiply(A=SquareRoot(None=@n1.A), B=Power(Base=@n10.A, Exp=Divide(B=@n1.A)))
  n13 = Add(
   A=Saturate(None=TextureSample() {samplerType=SAMPLERTYPE_LinearColor, texture=/Game/Weapons/FORT_Melee/Pickaxe_LiftingAura/FX/T_PaintClouds}.G)
   B=OneMinus(None=@n11.Dissolve 02)
  )
  Multiply(
   A=SmoothStep(Min=@n11.Dissolve 02, Max=Add(A=@n11.Dissolve 02), Value=Multiply(A=@n12, B=@n13))
   B=Ceil(None=Subtract(A=@n10.R, B=@n11.Dissolve))
  )
 ]
}
