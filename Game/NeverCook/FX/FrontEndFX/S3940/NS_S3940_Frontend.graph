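# Niagara front-end effect for S3940 that fills a wide box volume with drifting, fading pink heart sprites.
# Hearts is the denser emitter, with SubUV animation and a curve-weighted random size that also lowers peak alpha on larger sprites.
# Hearts002 adds a sparser, fainter layer with uniformly random sizes.
# Both spawn with randomized velocity and tilt, scale size over life, and declared User.Speed has no visible use.
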NiagaraSystem /Game/NeverCook/FX/FrontEndFX/S3940/NS_S3940_Frontend
dependencies {
 usedDataInterfaces [&NiagaraDataInterfaceCurve, &NiagaraDataInterfaceSpriteRendererInfo]
 usedDynamicInputs [&RandomRangeLinearColor]
 usedModules [&SystemState]
}
emitters {
 Hearts {
  emitterUpdateScript [
   &EmitterState(Life Cycle Mode=::NewEnumerator1, Loop Duration=1, UseLoopDelay=false)
   &SpawnRate(SpawnRate=18, Use Spawn Probability=false)
  ]
  particleSpawnScript [
   /SetVariables(Particles.RandomSize:NiagaraFloat=&FloatFromCurve(
     FloatCurve=DataInterface() {propertyValues {Curve {keys [<at 0 1>,<at 0.8017 2>,<at 1 4>]}}}
     CurveIndex=&RandomRangeFloat(Minimum=0, Maximum=1)
    )) {isSetParameters=true}
   &InitializeParticle(
    Color Minimum=<rgba 1 0.4588 0.732 1>
    Color Mode=::NewEnumerator2
    Lifetime=&RandomRangeFloat(Minimum=2, Maximum=3)
    Sprite Rotation Angle=&RandomRangeFloat(Minimum=-10, Maximum=10)
    Sprite Rotation Mode=::NewEnumerator0
    Sprite Size Mode=::NewEnumerator3
    Sprite UV Mode=::NewEnumerator1
    Uniform Sprite Size=&Multiply_Float(A=30, B=$Particles.RandomSize)
    UsePositionOffset=false
    Write Material Random=true
   )
   /SetVariables(Particles.SubImageIndex=&RandomRangeFloat(Minimum=0, Maximum=3)) {enabled=false, isSetParameters=true}
   &ShapeLocation(
    Box Size=<100 1500 500>
    MaskSpawnGroup=false
    Offset=<0 0 150>
    Offset Mode=::NewEnumerator0
    Override Seed=false
    Shape Primitive=::NewEnumerator2
    Surface Only=false
    UseOwnerScale=true
   )
   &AddVelocity(Velocity Speed Scale=&RandomRangeFloat(Minimum=0.7, Maximum=1.5))
   &SubUVAnimation(
    Evaluation Type=::NewEnumerator0
    "Sprite Renderer "=DataInterface() {propertyValues {SpriteRenderer=&self:Hearts_0.NiagaraSpriteRendererProperties_0}}
    SubUV Animation Mode:ENiagara_SubUVLookupModeV2=::NewEnumerator2
    UseEndFrame=false
    UseStartFrame=false
   )
  ]
  particleUpdateScript [
   &ParticleState()
   &ScaleSpriteSize(Uniform Curve Sprite Scale=DataInterface() {propertyValues {Curve {keys [
        {arriveTangent=2.4763, tangentMode=RCTM_User, time=0.2, value=1}
        {arriveTangent=2.5123, interpMode=RCIM_Cubic, leaveTangent=2.5123, tangentMode=RCTM_User, time=0.6, value=1}
        {arriveTangent=-0.5868, interpMode=RCIM_Cubic, leaveTangent=-0.5868, tangentMode=RCTM_User, time=1, value=2}
       ]}}}) {module=ScaleSpriteSize002}
   &SolveForcesAndVelocity(Clamp Velocity=false, Limit Acceleration=false)
   &ScaleColor(
    Scale Alpha=&FloatFromCurve(
     FloatCurve=DataInterface() {propertyValues {Curve {keys [
         {interpMode=RCIM_Cubic, tangentMode=RCTM_User, time=0, value=0}
         {interpMode=RCIM_Cubic, tangentMode=RCTM_User, time=0.5, value=1}
         {tangentMode=RCTM_User, time=1, value=0}
        ]}}}
     Scale Curve=&RemapRange(Input Value To Remap=$Particles.RandomSize, Input Min=1, Input Max=4, Output Min=0.7, Output Max=0.4)
    )
    ScaleA=true
    ScaleRGB=true
   )
  ]
  rendererClasses [&NiagaraSpriteRendererProperties]
  renderers [{rendererClass=&NiagaraSpriteRendererProperties}]
 }
 Hearts002 {
  emitterUpdateScript [
   &EmitterState(Life Cycle Mode=::NewEnumerator1, Loop Duration=1, UseLoopDelay=false)
   &SpawnRate(SpawnRate=6, Use Spawn Probability=false)
  ]
  particleSpawnScript [
   &InitializeParticle(
    Color=<rgba 1 0.1823 0.3387 1>
    Lifetime=&RandomRangeFloat(Minimum=2, Maximum=3)
    Sprite Rotation Angle=&RandomRangeFloat(Minimum=-10, Maximum=10)
    Sprite Rotation Mode=::NewEnumerator0
    Sprite Size Mode=::NewEnumerator3
    Sprite UV Mode=::NewEnumerator1
    Uniform Sprite Size=&RandomRangeFloat(Minimum=35, Maximum=55)
    UsePositionOffset=false
    Write Material Random=true
   )
   &ShapeLocation(
    Box Size=<100 1500 500>
    MaskSpawnGroup=false
    Offset=<0 0 150>
    Offset Mode=::NewEnumerator0
    Override Seed=false
    Shape Primitive=::NewEnumerator2
    Surface Only=false
    UseOwnerScale=true
   )
   &AddVelocity(Velocity Speed Scale=&RandomRangeFloat(Minimum=0.7, Maximum=1.5))
  ]
  particleUpdateScript [
   &ParticleState()
   &ScaleSpriteSize(Uniform Curve Sprite Scale=DataInterface() {propertyValues {Curve {keys [
        {arriveTangent=2.4763, tangentMode=RCTM_User, time=0.2, value=1.5}
        {arriveTangent=2.5123, interpMode=RCIM_Cubic, leaveTangent=2.5123, tangentMode=RCTM_User, time=0.6, value=1.5}
        {arriveTangent=-0.5868, interpMode=RCIM_Cubic, leaveTangent=-0.5868, tangentMode=RCTM_User, time=1, value=2}
       ]}}}) {module=ScaleSpriteSize002}
   &SolveForcesAndVelocity(Clamp Velocity=false, Limit Acceleration=false)
   &ScaleColor(
    Scale Alpha=&FloatFromCurve(
     FloatCurve=DataInterface() {propertyValues {Curve {keys [
         {interpMode=RCIM_Cubic, tangentMode=RCTM_User, time=0, value=0}
         {interpMode=RCIM_Cubic, tangentMode=RCTM_User, time=0.5, value=1}
         {tangentMode=RCTM_User, time=1, value=0}
        ]}}}
     Scale Curve=&RandomRangeFloat(Minimum=0.1, Maximum=0.3)
    )
    ScaleA=true
    ScaleRGB=true
   )
  ]
  rendererClasses [&NiagaraSpriteRendererProperties]
  renderers [{rendererClass=&NiagaraSpriteRendererProperties}]
 }
}
summary {
 systemName=NS_S3940_Frontend
 emitters [
  {emitterName=Hearts, rendererClasses [&NiagaraSpriteRendererProperties]}
  {emitterName=Hearts002, rendererClasses [&NiagaraSpriteRendererProperties]}
 ]
 userVariables [User.Speed {type=NiagaraFloat, defaultValue=0, description=""}]
}
